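; Spline-driven water edge actor for Asteria that builds cliff spline meshes and fits ocean water bodies to the edge.
; Cliff segments use short or long meshes by point spacing.
; Update rewrites two edge-mask and two cutout water splines, each covering half the loop, from offset spline points, then configures a child GeneratePlaneFromSpline actor for the ground plane.
; Editor events bake a scene-capture cutout texture and invoke an asset-saving helper.
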(event Collision|EventActorBeginOverlap
  (params (OtherActor :type Actor%20Object%20Reference)))

(event EventTick
  (params (DeltaSeconds :type Float%20%28single-precision%29)))

(event Custom|Update
  (Variables|Default|SetEditMode true)
  (CallFunction|UpdateOceanMaskSplinePoints)
  (CallFunction|UpdateMesh))

(event Custom|UpdateMesh
  (if (Variables|Default|GetEditMode)
    (if (Editor|InEditor)
      (bind _returnvalue (AddComponent|NODE_AddChildActorComponent-3 :RelativeTransform (Math|Transform|MakeTransform "0, 0, 0")))
      (bind _asgenerate_plane_from_spline (Utilities|Casting|CastToGeneratePlaneFromSpline (Variables|/Script/Engine.ChildActorComponent|GetChildActor _returnvalue)))
      (Variables|/Game/ArtTools/Procedural/WaterIslandEdge/GeneratePlaneFromSpline.GeneratePlaneFromSpline_C|SetTargetStaticMeshAsset (Variables|Default|GetGroundPlaneMeshAsset) _asgenerate_plane_from_spline)
      (Variables|/Game/ArtTools/Procedural/WaterIslandEdge/GeneratePlaneFromSpline.GeneratePlaneFromSpline_C|SetEnableLiveUpdate true _asgenerate_plane_from_spline)
      (Variables|/Game/ArtTools/Procedural/WaterIslandEdge/GeneratePlaneFromSpline.GeneratePlaneFromSpline_C|SetZDepthPosition (Variables|Default|GetZDepth%20Position) _asgenerate_plane_from_spline)
      (Variables|/Game/ArtTools/Procedural/WaterIslandEdge/GeneratePlaneFromSpline.GeneratePlaneFromSpline_C|SetOffset (Variables|Default|GetOffset) _asgenerate_plane_from_spline)
      (Variables|/Game/ArtTools/Procedural/WaterIslandEdge/GeneratePlaneFromSpline.GeneratePlaneFromSpline_C|SetStealSplineActor self _asgenerate_plane_from_spline)
      (Variables|/Game/ArtTools/Procedural/WaterIslandEdge/GeneratePlaneFromSpline.GeneratePlaneFromSpline_C|SetSplineOuter (Variables|Default|GetSpline) _asgenerate_plane_from_spline)
      (Variables|/Game/ArtTools/Procedural/WaterIslandEdge/GeneratePlaneFromSpline.GeneratePlaneFromSpline_C|SetSplineInnerCutout (Variables|Default|GetSplineInnerCutout) _asgenerate_plane_from_spline))
    (else
      (Variables|Default|SetEditMode true))))

(event Custom|RenderTexture
  (bind _scenecapturecomponent2d (Variables|Default|GetSceneCaptureComponent2D))
  (Utilities|FlowControl|Sequence
    (:then_0
      (bind _returnvalue (AddComponent|NODE_AddChildActorComponent-5 :RelativeTransform (Math|Transform|MakeTransform "0, 0, 0")))
      (bind _asstatic_mesh_actor (Utilities|Casting|CastToStaticMeshActor (Variables|/Script/Engine.ChildActorComponent|GetChildActor _returnvalue)))
      (Components|StaticMesh|SetStaticMesh (Variables|/Script/Engine.StaticMeshActor|GetStaticMeshComponent _asstatic_mesh_actor) (Variables|Default|GetGroundPlaneMeshAsset)))
    (:then_1
      (Rendering|SetActorHiddenInGame (Variables|/Script/Engine.ChildActorComponent|GetChildActor _returnvalue) false)
      (Variables|/Script/Engine.SceneCaptureComponent|SetShowOnlyActors (Utilities|Array|MakeArray (Variables|/Script/Engine.ChildActorComponent|GetChildActor _returnvalue)) _scenecapturecomponent2d)
      (Rendering|SceneCapture|CaptureScene _scenecapturecomponent2d :!class /Script/Engine.SceneCaptureComponent2D)
      (EditorScripting|Rendering|ConvertRenderTargettoTexture2DEditorOnly (Variables|/Script/Engine.SceneCaptureComponent2D|GetTextureTarget _scenecapturecomponent2d) (Variables|Default|GetWaterCutoutTextureAsset))
      (Rendering|SetActorHiddenInGame (Variables|/Script/Engine.ChildActorComponent|GetChildActor _returnvalue) true))
    (:then_2
      (Actor|DestroyActor (Variables|/Script/Engine.ChildActorComponent|GetChildActor _returnvalue)))))

(event Custom|Save%20all%20Assets
  (bind (_pathpart _filenamepart) (Utilities|Paths|Split (Utilities|GetObjectPathString (Variables|Default|GetWaterCutoutTextureAsset))))
  (bind (_pathpart_2 _filenamepart_2) (Utilities|Paths|Split (Utilities|GetObjectPathString (Variables|Default|GetGroundPlaneMeshAsset))))
  (bind _returnvalue (Game|SpawnActorfromClass "/Game/ArtTools/Procedural/WaterIslandEdge/SaveWaterEdgeAssets.SaveWaterEdgeAssets_C" (Math|Transform|MakeTransform "0, 0, 0") "AlwaysSpawn" "OverrideRootScale" 0 (Utilities|String|Append _pathpart "/" _filenamepart) (Utilities|String|Append _pathpart_2 "/" _filenamepart_2) 0))
  (Actor|DestroyActor _returnvalue))

(fn Add_SplineMesh_Components
  (params (Evaluate%20WPO :type Boolean))
  (bind _localsplinecomponentlist (LocalVariables|GetLocalSplineComponentlist))
  (Utilities|FlowControl|Sequence
    (:then_0
      (for _index (range (Spline|GetNumberOfSplinePoints (Variables|Default|GetSpline)))
        (LocalVariables|SetCurrentPoint _index)
        (bind _returnvalue (AddComponent|NODE_AddSplineMeshComponent-2 :RelativeTransform (Math|Transform|MakeTransform "0, 0, 0")))
        (Components|StaticMesh|SetStaticMesh _returnvalue (select (< (Math|Vector|Distance(Vector) (Spline|GetLocationatSplinePoint (Variables|Default|GetSpline) _index "Local") (Spline|GetLocationatSplinePoint (Variables|Default|GetSpline) (+ _index 1) "Local")) (Variables|Default|GetSmallAssetDistance)) "/Game/Athena/Asteria/Environments/Water/WaterEdge/SM_WaterEdgeCliff_Short.SM_WaterEdgeCliff_Short" "/Game/Athena/Asteria/Environments/Water/WaterEdge/SM_WaterEdgeCliff.SM_WaterEdgeCliff"))
        (Rendering|LOD|SetEvaluateWorldPositionOffset _returnvalue Evaluate%20WPO :!class /Script/Engine.StaticMeshComponent)
        (Default|SetSplineMeshCollision _returnvalue)
        (bind _output_get (LocalVariables|SetCurrentSplineComponent _returnvalue))
        (Utilities|Array|Add _localsplinecomponentlist _output_get)
        (Utilities|FlowControl|Sequence
          (:then_1
            (Default|SetLocationfromSpline (LocalVariables|GetCurrentSplineComponent) (LocalVariables|GetCurrentPoint))
            (SplineMesh|SetForwardAxis (LocalVariables|GetCurrentSplineComponent) "X" true)))))
    (:then_1
      (return _localsplinecomponentlist))))

(fn AlginSplineToOcean
  (bind _spline (Variables|Default|GetSpline))
  (if (Variables|Default|GetAlignToOcean)
    (bind _outactors (Actor|GetAllActorsOfClass "/Game/Athena/Asteria/Environments/Water/Asteria_Waterbody_Ocean_Parent.Asteria_Waterbody_Ocean_Parent_C"))
    (for _array_element _outactors
      (for _index (range (Spline|GetNumberOfSplinePoints _spline))
        (Spline|SetLocationatSplinePoint _spline _index (Math|Vector|MakeVector (.x (Spline|GetLocationatSplinePoint _spline _index "World")) (.y (Spline|GetLocationatSplinePoint _spline _index "World")) (.z (Transformation|GetActorLocation self))) "World" true)
        (Spline|SetTangentsatSplinePoint _spline _index (* (Spline|GetArriveTangentatSplinePoint _spline _index "World") "1.000000,1.000000,0") (* (Spline|GetLeaveTangentatSplinePoint _spline _index "World") "1.000000,1.000000,0") "World" true)))))

(fn AutoTangentSpline
  (bind _spline (Variables|Default|GetSpline))
  (for _index (range (Spline|GetNumberOfSplinePoints _spline))
    (Spline|SetTangentatSplinePoint (Variables|Default|GetSpline) _index (Math|Vector|SelectVector (* (Math|Vector|Normalize (.Tangent (Spline|GetLocationandTangentatSplinePoint _spline _index "Local"))) (* (Math|Vector|Distance(Vector) (Spline|GetLocationandTangentatSplinePoint _spline _index "Local") (Spline|GetLocationandTangentatSplinePoint _spline (+ _index 1) "Local")) (Variables|Default|GetTension))) (.Tangent (Spline|GetLocationandTangentatSplinePoint _spline _index "Local")) (Variables|Default|GetAutoSmoothTangent)) "Local" true)))

(fn clearMeshes
  (for _array_element (Actor|GetComponentsByClass "/Script/Engine.SplineMeshComponent")
    (Components|DestroyComponent _array_element)))

(fn ConvertSplineDataLocation
  (params (Location :type Vector) (splinemesh :type Scene%20Component%20Object%20Reference))
  (return (Math|Transform|InverseTransformLocation (Transformation|GetWorldTransform splinemesh) Location)))

(fn ConvertSplineDataTangent
  (params (Tangent :type Vector) (splinemesh :type Scene%20Component%20Object%20Reference))
  (return (* (Math|Vector|Normalize (Math|Transform|InverseTransformDirection (Transformation|GetWorldTransform splinemesh) Tangent)) (Variables|Default|GettangentMult))))

(fn getInfoFromSplineDistance
  (params (distance :type Float%20%28double-precision%29))
  (bind _newlocalvar (LocalVariables|GetNewLocalVar))
  (LocalVariables|Setl_loc (Spline|GetLocationatDistanceAlongSpline (Variables|Default|GetSpline) distance _newlocalvar))
  (LocalVariables|Setl_rot (Spline|GetRotationatDistanceAlongSpline (Variables|Default|GetSpline) (Variables|Getdistance) _newlocalvar))
  (LocalVariables|Setl_tangentStart (Spline|GetTangentatDistanceAlongSpline (Variables|Default|GetSpline) (Variables|Getdistance) _newlocalvar))
  (return (LocalVariables|Getl_loc) (LocalVariables|Getl_rot) (LocalVariables|Getl_tangentStart)))

(fn RegenerateMeshEdges
  (bind (_l_firstpoint_location_2_759a89cc40091966b22b55bdc810b244 _l_firstpoint_rotation_8_06bf9b12449a60eefb9cb6bec2ccc130 _l_firstpoint_tangent_6_cb194ddf46c7fb2c6bf90fa8b8c2b068) (Utilities|Struct|BreakWaterEdgeBuilderStruct (Variables|Getl_firstPoint_location_2_759A89CC40091966B22B55BDC810B244)))
  (bind (_l_lastpoint_location_2_759a89cc40091966b22b55bdc810b244 _l_lastpoint_rotation_8_06bf9b12449a60eefb9cb6bec2ccc130 _l_lastpoint_tangent_6_cb194ddf46c7fb2c6bf90fa8b8c2b068) (Utilities|Struct|BreakWaterEdgeBuilderStruct (Variables|Getl_lastPoint_location_2_759A89CC40091966B22B55BDC810B244)))
  (LocalVariables|Setl_length (Spline|GetSplineLength (Variables|Default|GetSpline)))
  (Utilities|FlowControl|Sequence
    (:then_0
      (CallFunction|ClearMeshes))
    (:then_1
      (bind _output_get (LocalVariables|Setl_meshWidth 4e+04))
      (LocalVariables|Setl_finalWidth (* _output_get (Variables|Default|GetMeshScaleMult))))
    (:then_2
      (if (> (LocalVariables|Getl_length) 0)
        (bind (_outlocation _outrotation _outtangent) (CallFunction|GetInfofromSplineDistance 0.0))
        (LocalVariables|Setl_firstPoint (Utilities|Struct|MakeWaterEdgeBuilderStruct _outlocation _outrotation _outtangent))
        (while (< (LocalVariables|Getl_curDistance) (LocalVariables|Getl_length))
          (Utilities|FlowControl|Sequence
            (:then_0
              (bind _output_get_2 (LocalVariables|Setl_targetDistance (+ (LocalVariables|Getl_curDistance) (LocalVariables|Getl_finalWidth))))
              (bind (_outlocation_2 _outrotation_2 _outtangent_2) (CallFunction|GetInfofromSplineDistance _output_get_2))
              (Variables|Setl_lastPoint_location_2_759A89CC40091966B22B55BDC810B244 (Utilities|Struct|MakeWaterEdgeBuilderStruct _outlocation_2 _outrotation_2 _outtangent_2)))
            (:then_1
              (CallFunction|SpawnEdgeMesh _l_firstpoint_location_2_759a89cc40091966b22b55bdc810b244 _l_firstpoint_rotation_8_06bf9b12449a60eefb9cb6bec2ccc130 _l_firstpoint_tangent_6_cb194ddf46c7fb2c6bf90fa8b8c2b068 _l_lastpoint_location_2_759a89cc40091966b22b55bdc810b244 _l_lastpoint_tangent_6_cb194ddf46c7fb2c6bf90fa8b8c2b068))
            (:then_2
              (LocalVariables|Setl_curDistance (LocalVariables|Getl_targetDistance))
              (LocalVariables|Setl_firstPoint (LocalVariables|Getl_lastPoint)))))))))

(fn SetLocationFromSpline
  (params (SplineComponant :type Spline%20Mesh%20Component%20Object%20Reference) (CurrentPoint :type Integer))
  (bind _returnvalue (+ CurrentPoint 1))
  (bind _spline (Variables|Default|GetSpline))
  (bind (_location _tangent) (Spline|GetLocationandTangentatSplinePoint _spline _returnvalue "Local"))
  (bind (_location_2 _tangent_2) (Spline|GetLocationandTangentatSplinePoint _spline CurrentPoint "Local"))
  (bind _splinecomponant (Variables|GetSplineComponant))
  (bind _meshscalemult (Variables|Default|GetMeshScaleMult))
  (Utilities|FlowControl|Sequence
    (:then_0
      (LocalVariables|SetCurrenPointLocation _location_2)
      (LocalVariables|SetCurrentTangent _tangent_2)
      (LocalVariables|SetNextPointLocation _location)
      (LocalVariables|SetNextTangent _tangent)
      (SplineMesh|SetStartandEnd SplineComponant (LocalVariables|GetCurrenPointLocation) (Math|Vector|SelectVector (* (Math|Vector|Normalize (LocalVariables|GetCurrentTangent)) (* (Math|Vector|Distance(Vector) _location_2 _location) (Variables|Default|GetTension))) (LocalVariables|GetCurrentTangent) (Variables|Default|GetAutoSmoothTangent)) (LocalVariables|GetNextPointLocation) (Math|Vector|SelectVector (* (Math|Vector|Normalize (LocalVariables|GetNextTangent)) (* (Math|Vector|Distance(Vector) _location_2 _location) (Variables|Default|GetTension))) (LocalVariables|GetNextTangent) (Variables|Default|GetAutoSmoothTangent)) true :!class /Script/Engine.SplineMeshComponent)
      (SplineMesh|SetStartScale _splinecomponant (Math|Vector2D|MakeVector2D _meshscalemult _meshscalemult) true :!class /Script/Engine.SplineMeshComponent)
      (SplineMesh|SetEndScale _splinecomponant (Math|Vector2D|MakeVector2D _meshscalemult _meshscalemult) true :!class /Script/Engine.SplineMeshComponent))
    (:then_1
      (return (.y (Spline|GetScaleatSplinePoint (Variables|Default|GetSpline) CurrentPoint)) (.y (Spline|GetScaleatSplinePoint (Variables|Default|GetSpline) _returnvalue))))))

(fn SetSplineMeshCollision
  (params (SplineComponent :type Spline%20Mesh%20Component%20Object%20Reference))
  (if (Variables|Default|GetCollision)
    (Collision|SetCollisionEnabled SplineComponent "QueryAndPhysics")
    (Collision|SetCollisionResponsetoChannel SplineComponent "ECC_GameTraceChannel11" "ECR_Block") ; Block foot step traces for sound
    (else
      (Collision|SetCollisionEnabled SplineComponent "NoCollision"))))

(fn spawnEdgeMesh
  (params (Location :type Vector) (Rotation :type Rotator) (StartTangent :type Vector) (EndPos :type Vector) (EndTangent :type Vector) (scaleWidth :type Float%20%28double-precision%29))
  (bind _l_newmesh (LocalVariables|Getl_newMesh))
  (bind _l_newmesh_2 (LocalVariables|Getl_newMesh))
  (bind _returnvalue (AddComponent|NODE_AddSplineMeshComponent-0))
  (bind _output_get (LocalVariables|Setl_newMesh _returnvalue))
  (Transformation|SetWorldTransform _output_get (Math|Transform|MakeTransform Location Rotation) false true)
  (Utilities|FlowControl|Sequence
    (:then_0
      (SplineMesh|SetStartPosition (LocalVariables|Getl_newMesh) (CallFunction|ConvertSplineDataLocation Location _l_newmesh) false)
      (SplineMesh|SetStartTangent (LocalVariables|Getl_newMesh) (CallFunction|ConvertSplineDataTangent StartTangent _l_newmesh) false :!class /Script/Engine.SplineMeshComponent))
    (:then_1
      (SplineMesh|SetEndPosition (LocalVariables|Getl_newMesh) (CallFunction|ConvertSplineDataLocation EndPos _l_newmesh_2) false)
      (SplineMesh|SetEndTangent (LocalVariables|Getl_newMesh) (CallFunction|ConvertSplineDataTangent EndTangent _l_newmesh_2) false :!class /Script/Engine.SplineMeshComponent))
    (:then_2
      (SplineMesh|UpdateMesh (LocalVariables|Getl_newMesh)))
    (:then_3
      (SplineMesh|SetStartScale _returnvalue "(X=2.000000,Y=2.000000)" true :!class /Script/Engine.SplineMeshComponent)
      (SplineMesh|SetEndScale _returnvalue "(X=2.000000,Y=2.000000)" true :!class /Script/Engine.SplineMeshComponent))))

(fn UpdateOceanMaskSplinePoints
  (bind _oceanwatercutout_2 (Variables|Default|GetOceanWaterCutOut_2))
  (bind _returnvalue (Actor|GetComponentbyClass _oceanwatercutout_2 "/Script/Water.WaterSplineComponent"))
  (bind _spline (Variables|Default|GetSpline))
  (bind _returnvalue_2 (Spline|GetNumberOfSplinePoints _spline))
  (bind _oceanwatercutout (Variables|Default|GetOceanWaterCutOut))
  (bind _returnvalue_3 (Actor|GetComponentbyClass _oceanwatercutout "/Script/Water.WaterSplineComponent"))
  (bind _spline_2 (Variables|Default|GetSpline))
  (bind _returnvalue_4 (Spline|GetNumberOfSplinePoints _spline_2))
  (bind _spline_3 (Variables|Default|GetSpline))
  (bind _returnvalue_5 (Spline|GetNumberOfSplinePoints _spline_3))
  (bind _returnvalue_6 (/ _returnvalue_5 2))
  (bind _location (Spline|GetLocationandTangentatSplinePoint _spline_3 _returnvalue_6 "World"))
  (bind _returnvalue_7 (Spline|GetTangentatSplinePoint _spline_3 _returnvalue_6 "World"))
  (bind _returnvalue_8 (Math|Vector|CrossProduct _returnvalue_7 "0, 0,1.000000"))
  (bind _returnvalue_9 (Math|Vector|Normalize _returnvalue_8))
  (bind _spline_4 (Variables|Default|GetSpline))
  (bind _returnvalue_10 (Spline|GetTangentatSplinePoint _spline_4 0 "World"))
  (bind _oceanmaskoutercutoutoffset (Variables|Default|GetOceanMaskOuterCutoutOffset))
  (bind _oceanmaskouteroffset (Variables|Default|GetOceanMaskOuterOffset))
  (bind _returnvalue_11 (+ _oceanmaskouteroffset _oceanmaskoutercutoutoffset))
  (bind _returnvalue_12 (* _returnvalue_9 _returnvalue_11))
  (bind _returnvalue_13 (- _location _returnvalue_12))
  (bind _x (.x _returnvalue_13))
  (bind _y (.y _returnvalue_13))
  (bind _returnvalue_14 (Math|Float|Sign(Float) _y))
  (bind _waterzoneactor (Variables|Default|GetWaterZoneActor))
  (bind _returnvalue_15 (Transformation|GetActorLocation _waterzoneactor))
  (bind _x_2 (.x _returnvalue_15))
  (bind _y_2 (.y _returnvalue_15))
  (bind _spline_5 (Variables|Default|GetSpline))
  (bind _returnvalue_16 (Spline|GetNumberOfSplinePoints _spline_5))
  (bind _waterzoneactor_2 (Variables|Default|GetWaterZoneActor))
  (bind _returnvalue_17 (Transformation|GetActorLocation _waterzoneactor_2))
  (bind _zoneextent (Variables|/Script/Water.WaterZone|GetZoneExtent _waterzoneactor_2))
  (bind _x_3 (Math|Vector2D|BreakVector2D _zoneextent))
  (bind _x_4 (.x _returnvalue_17))
  (bind _y_3 (.y _returnvalue_17))
  (bind _spline_6 (Variables|Default|GetSpline))
  (bind _returnvalue_18 (Spline|GetNumberOfSplinePoints _spline_6))
  (bind _returnvalue_19 (/ _returnvalue_18 2))
  (bind _returnvalue_20 (Spline|GetTangentatSplinePoint _spline_6 _returnvalue_19 "World"))
  (bind _oceanwateredgemaskactor_2 (Variables|Default|GetOceanWaterEdgeMaskActor_2))
  (bind _returnvalue_21 (Actor|GetComponentbyClass _oceanwateredgemaskactor_2 "/Script/Water.WaterSplineComponent"))
  (bind _spline_7 (Variables|Default|GetSpline))
  (bind _returnvalue_22 (Spline|GetNumberOfSplinePoints _spline_7))
  (bind _zoneextent_2 (Variables|/Script/Water.WaterZone|GetZoneExtent _waterzoneactor))
  (bind _x_5 (Math|Vector2D|BreakVector2D _zoneextent_2))
  (bind _spline_8 (Variables|Default|GetSpline))
  (bind _location_2 (Spline|GetLocationandTangentatSplinePoint _spline_8 0 "World"))
  (bind _returnvalue_23 (Spline|GetTangentatSplinePoint _spline_8 0 "World"))
  (bind _returnvalue_24 (Math|Vector|CrossProduct _returnvalue_23 "0, 0,1.000000"))
  (bind _returnvalue_25 (Math|Vector|Normalize _returnvalue_24))
  (bind _oceanmaskoutercutoutoffset_2 (Variables|Default|GetOceanMaskOuterCutoutOffset))
  (bind _oceanmaskouteroffset_2 (Variables|Default|GetOceanMaskOuterOffset))
  (bind _returnvalue_26 (+ _oceanmaskouteroffset_2 _oceanmaskoutercutoutoffset_2))
  (bind _returnvalue_27 (* _returnvalue_25 _returnvalue_26))
  (bind _returnvalue_28 (- _location_2 _returnvalue_27))
  (bind _x_6 (.x _returnvalue_28))
  (bind _y_4 (.y _returnvalue_28))
  (bind _returnvalue_29 (Math|Float|Sign(Float) _y_4))
  (bind _oceanwateredgemaskactor (Variables|Default|GetOceanWaterEdgeMaskActor))
  (bind _returnvalue_30 (Actor|GetComponentbyClass _oceanwateredgemaskactor "/Script/Water.WaterSplineComponent"))
  (bind _returnvalue_31 (/ _x_5 2.0))
  (bind _returnvalue_32 (- _y_2 _returnvalue_31))
  (bind _returnvalue_33 (/ _returnvalue_16 2))
  (bind _returnvalue_34 (+ _y_2 _returnvalue_31))
  (bind _returnvalue_35 (/ _x_3 2.0))
  (bind _returnvalue_36 (- _y_3 _returnvalue_35))
  (bind _returnvalue_37 (- _x_4 _returnvalue_35))
  (bind _returnvalue_38 (- _returnvalue_22 0))
  (bind _waterzoneoffsetmargin (Variables|Default|GetWaterZoneOffsetMargin))
  (bind _returnvalue_39 (- _returnvalue_36 _waterzoneoffsetmargin))
  (bind _waterzoneoffsetmargin_2 (Variables|Default|GetWaterZoneOffsetMargin))
  (bind _returnvalue_40 (- _returnvalue_32 _waterzoneoffsetmargin_2))
  (bind _returnvalue_41 (+ _returnvalue_34 _waterzoneoffsetmargin_2))
  (bind _returnvalue_42 (+ _x_2 _returnvalue_31))
  (bind _returnvalue_43 (+ _returnvalue_42 _waterzoneoffsetmargin_2))
  (bind _returnvalue_44 (> _x _y))
  (bind _returnvalue_45 (> _x_6 _y_4))
  (bind _returnvalue_46 (/ _returnvalue_4 2))
  (bind _returnvalue_47 (- _returnvalue_37 _waterzoneoffsetmargin))
  (bind _spline_9 (Variables|Default|GetSpline))
  (bind _splinewateredgemask (LocalVariables|GetSplineWaterEdgeMask))
  (bind _spline_10 (Variables|Default|GetSpline))
  (bind _splinewateredgemask_2 (LocalVariables|GetSplineWaterEdgeMask))
  (Utilities|FlowControl|Sequence
    (:then_0
      ;[1 1 1] 1stHalf
      (Utilities|IsValid (Variables|Default|GetOceanWaterEdgeMaskActor)
        (:Is%20Valid
          (Utilities|IsValid _returnvalue_30
            (:Is%20Valid
              (LocalVariables|SetSplineWaterEdgeMask _returnvalue_30)
              (Transformation|SetActorLocation _oceanwateredgemaskactor "0.000000,0.000000,-52.000000" false false)
              (UEFNConversionHelper|SetComponentHiddenEditor _returnvalue_30 (not (Variables|Default|GetDebugSplines)))
              (Spline|ClearSplinePoints _returnvalue_30 true)
              (Utilities|FlowControl|Sequence
                (:then_0
                  (Utilities|FlowControl|Sequence
                    (:then_0
                      (Spline|AddSplinePoint (LocalVariables|GetSplineWaterEdgeMask) (Math|Vector|MakeVector (.x (Spline|GetLocationatSplinePoint (Variables|Default|GetSpline) 0 "World")) _returnvalue_39) "World" true))
                    (:then_1
                      (Spline|AddSplinePoint (LocalVariables|GetSplineWaterEdgeMask) (Math|Vector|MakeVector _returnvalue_47 _returnvalue_39) "World" true))
                    (:then_2
                      (Spline|AddSplinePoint (LocalVariables|GetSplineWaterEdgeMask) (Math|Vector|MakeVector _returnvalue_47 (+ (+ _y_3 _returnvalue_35) _waterzoneoffsetmargin)) "World" true))
                    (:then_3
                      (Spline|AddSplinePoint (LocalVariables|GetSplineWaterEdgeMask) (Math|Vector|MakeVector (.x (Spline|GetLocationatSplinePoint (Variables|Default|GetSpline) (/ (Spline|GetNumberOfSplinePoints (Variables|Default|GetSpline)) 2) "World")) (+ (+ _y_3 _returnvalue_35) _waterzoneoffsetmargin)) "World" true)))
                  (Spline|SetSplinePointType (LocalVariables|GetSplineWaterEdgeMask) (- (Spline|GetNumberOfSplinePoints (LocalVariables|GetSplineWaterEdgeMask)) 1) "Linear" true))
                (:then_1
                  (for _index (range (+ _returnvalue_33 1))
                    (Spline|AddSplinePoint _splinewateredgemask_2 (- (Spline|GetLocationandTangentatSplinePoint (Variables|Default|GetSpline) (- _returnvalue_33 _index) "World") (* (Math|Vector|Normalize (Math|Vector|CrossProduct (Spline|GetTangentatSplinePoint (Variables|Default|GetSpline) (- _returnvalue_33 _index) "World") "0, 0,1.000000")) (Variables|Default|GetOceanMaskOuterOffset))) "World" true)
                    (Spline|SetTangentsatSplinePoint (LocalVariables|GetSplineWaterEdgeMask) (- (Spline|GetNumberOfSplinePoints _splinewateredgemask_2) 1) (Math|Vector|NegateVector (Spline|GetLeaveTangentatSplinePoint (Variables|Default|GetSpline) (- _returnvalue_33 _index) "World")) (Math|Vector|NegateVector (Spline|GetArriveTangentatSplinePoint (Variables|Default|GetSpline) (- _returnvalue_33 _index) "World")) "World" true)))
                (:then_2
                  (bind _aswater_body (Utilities|Casting|CastToWaterBody _oceanwateredgemaskactor))
                  (Water|OnWaterBodyChanged (Variables|/Script/Water.WaterBody|GetWaterBodyComponent _aswater_body) true true false))))))))
    (:then_1
      ;[1 1 1] 2nd Half
      (Utilities|IsValid (Variables|Default|GetOceanWaterEdgeMaskActor_2)
        (:Is%20Valid
          (Utilities|IsValid _returnvalue_21
            (:Is%20Valid
              (LocalVariables|SetSplineWaterEdgeMask _returnvalue_21)
              (Transformation|SetActorLocation _oceanwateredgemaskactor_2 "0.000000,0.000000,-52.000000" false false)
              (UEFNConversionHelper|SetComponentHiddenEditor _returnvalue_21 (not (Variables|Default|GetDebugSplines)))
              (Spline|ClearSplinePoints _returnvalue_21 true)
              (Utilities|FlowControl|Sequence
                (:then_0
                  (Utilities|FlowControl|Sequence
                    (:then_0
                      (Spline|AddSplinePoint (LocalVariables|GetSplineWaterEdgeMask) (Math|Vector|MakeVector (.x (Spline|GetLocationatSplinePoint (Variables|Default|GetSpline) (/ (Spline|GetNumberOfSplinePoints (Variables|Default|GetSpline)) 2) "World")) _returnvalue_41) "World" true))
                    (:then_1
                      (Spline|AddSplinePoint (LocalVariables|GetSplineWaterEdgeMask) (Math|Vector|MakeVector _returnvalue_43 _returnvalue_41) "World" true))
                    (:then_2
                      (Spline|AddSplinePoint (LocalVariables|GetSplineWaterEdgeMask) (Math|Vector|MakeVector _returnvalue_43 _returnvalue_40) "World" true))
                    (:then_3
                      (Spline|AddSplinePoint (LocalVariables|GetSplineWaterEdgeMask) (Math|Vector|MakeVector (.x (Spline|GetLocationatSplinePoint (Variables|Default|GetSpline) 0 "World")) _returnvalue_40) "World" true)))
                  (Spline|SetSplinePointType (LocalVariables|GetSplineWaterEdgeMask) (- (Spline|GetNumberOfSplinePoints (LocalVariables|GetSplineWaterEdgeMask)) 1) "Linear" true))
                (:then_1
                  (for _index_2 (range (+ (+ (/ _returnvalue_22 2) 1) 1))
                    (Spline|AddSplinePoint _splinewateredgemask (- (Spline|GetLocationandTangentatSplinePoint (Variables|Default|GetSpline) (- _returnvalue_38 _index_2) "World") (* (Math|Vector|Normalize (Math|Vector|CrossProduct (Spline|GetTangentatSplinePoint (Variables|Default|GetSpline) (- _returnvalue_38 _index_2) "World") "0, 0,1.000000")) (Variables|Default|GetOceanMaskOuterOffset))) "World" true)
                    (Spline|SetTangentsatSplinePoint (LocalVariables|GetSplineWaterEdgeMask) (- (Spline|GetNumberOfSplinePoints _splinewateredgemask) 1) (Math|Vector|NegateVector (Spline|GetLeaveTangentatSplinePoint (Variables|Default|GetSpline) (- _returnvalue_38 _index_2) "World")) (Math|Vector|NegateVector (Spline|GetArriveTangentatSplinePoint (Variables|Default|GetSpline) (- _returnvalue_38 _index_2) "World")) "World" true)))
                (:then_2
                  (bind _aswater_body_2 (Utilities|Casting|CastToWaterBody _oceanwateredgemaskactor_2))
                  (Water|OnWaterBodyChanged (Variables|/Script/Water.WaterBody|GetWaterBodyComponent _aswater_body_2) true true false))))))))
    (:then_2
      (Utilities|IsValid (Variables|Default|GetOceanWaterCutOut)
        (:Is%20Valid
          (Utilities|IsValid _returnvalue_3
            (:Is%20Valid
              (LocalVariables|SetSplineWaterCutOut _returnvalue_3)
              (Transformation|SetActorLocation _oceanwatercutout "0, 0, 0" false false)
              (UEFNConversionHelper|SetComponentHiddenEditor _returnvalue_3 (not (Variables|Default|GetDebugSplines)))
              (Utilities|FlowControl|Sequence
                (:then_0
                  (Spline|ClearSplinePoints (LocalVariables|GetSplineWaterCutOut) true)
                  (for _index_3 (range (+ _returnvalue_46 1))
                    (Spline|AddSplinePoint (LocalVariables|GetSplineWaterCutOut) (- (Spline|GetLocationandTangentatSplinePoint _spline_9 (Math|Integer|SelectInt _index_3 _index_3 (== _index_3 _returnvalue_46)) "World") (* (Math|Vector|Normalize (Math|Vector|CrossProduct (Spline|GetTangentatSplinePoint (Variables|Default|GetSpline) (Math|Integer|SelectInt _index_3 _index_3 (== _index_3 _returnvalue_46)) "World") "0, 0,1.000000")) (Variables|Default|GetOceanMaskOuterOffset))) "World" true)
                    (Spline|SetTangentsatSplinePoint (LocalVariables|GetSplineWaterCutOut) _index_3 (Spline|GetArriveTangentatSplinePoint _spline_9 (Math|Integer|SelectInt _index_3 _index_3 (== _index_3 _returnvalue_46)) "World") (Spline|GetLeaveTangentatSplinePoint _spline_9 (Math|Integer|SelectInt _index_3 _index_3 (== _index_3 _returnvalue_46)) "World") "World" true)))
                (:then_1
                  (Spline|AddSplinePoint (LocalVariables|GetSplineWaterCutOut) _returnvalue_13 "World" true)
                  (Spline|SetTangentsatSplinePoint (LocalVariables|GetSplineWaterCutOut) 0 (Spline|GetArriveTangentatSplinePoint (Variables|Default|GetSpline) _returnvalue_6 "World") (Spline|GetLeaveTangentatSplinePoint (Variables|Default|GetSpline) _returnvalue_6 "World") "World" true)
                  (LocalVariables|SetMaxDistance (Math|Float|Max(Float) (Math|Float|Absolute(Float) _x) (Math|Float|Absolute(Float) _y)))
                  (LocalVariables|SetZHeight (.z _returnvalue_13))
                  (LocalVariables|SetX_Sign (Math|Float|SelectFloat _returnvalue_14 (Math|Float|Sign(Float) _x) _returnvalue_44))
                  (LocalVariables|SetY_Sign (Math|Float|SelectFloat _returnvalue_14 (* _returnvalue_14 -1.0) _returnvalue_44)))
                (:then_2
                  (Utilities|FlowControl|Sequence
                    (:then_0
                      _)
                    (:then_3
                      (Spline|AddSplinePoint _returnvalue_3 (Math|Vector|MakeVector (* (LocalVariables|GetMaxDistance) (LocalVariables|GetX_Sign)) (* (LocalVariables|GetMaxDistance) (* (LocalVariables|GetY_Sign) -1.0)) (LocalVariables|GetZHeight)) "World" true))
                    (:then_4
                      (Spline|AddSplinePoint _returnvalue_3 (Math|Vector|MakeVector (* (LocalVariables|GetMaxDistance) (LocalVariables|GetX_Sign)) (* (LocalVariables|GetMaxDistance) (LocalVariables|GetY_Sign)) (LocalVariables|GetZHeight)) "World" true)))
                  (Spline|SetTangentsatSplinePoint _returnvalue_3 (- (Spline|GetNumberOfSplinePoints _returnvalue_3) 1) (Math|Vector|MakeVector 0.0) (Math|Vector|MakeVector 0.0) "World" true))
                (:then_3
                  (Spline|AddSplinePoint _returnvalue_3 (- (Spline|GetLocationandTangentatSplinePoint _spline_4 0 "World") (* (Math|Vector|Normalize (Math|Vector|CrossProduct _returnvalue_10 "0, 0,1.000000")) (+ (Variables|Default|GetOceanMaskOuterOffset) (Variables|Default|GetOceanMaskOuterCutoutOffset)))) "World" true)
                  (Spline|SetTangentatSplinePoint _returnvalue_3 0 _returnvalue_10 "World" true))
                (:then_4
                  (bind _aswater_body_3 (Utilities|Casting|CastToWaterBody _oceanwatercutout))
                  (Water|OnWaterBodyChanged (Variables|/Script/Water.WaterBody|GetWaterBodyComponent _aswater_body_3) true true false))))))))
    (:then_3
      (Utilities|IsValid (Variables|Default|GetOceanWaterCutOut_2)
        (:Is%20Valid
          (Utilities|IsValid _returnvalue
            (:Is%20Valid
              (LocalVariables|SetSplineWaterCutOut _returnvalue)
              (Transformation|SetActorLocation _oceanwatercutout_2 "0, 0, 0" false false)
              (UEFNConversionHelper|SetComponentHiddenEditor _returnvalue (not (Variables|Default|GetDebugSplines)))
              (Utilities|FlowControl|Sequence
                (:then_0
                  (Spline|ClearSplinePoints (LocalVariables|GetSplineWaterCutOut) true)
                  (for _index_4 (range (/ _returnvalue_2 2) (+ _returnvalue_2 1))
                    (Spline|AddSplinePoint (LocalVariables|GetSplineWaterCutOut) (- (Spline|GetLocationandTangentatSplinePoint _spline_10 (Math|Integer|SelectInt 0 _index_4 (== _index_4 _returnvalue_2)) "World") (* (Math|Vector|Normalize (Math|Vector|CrossProduct (Spline|GetTangentatSplinePoint (Variables|Default|GetSpline) (Math|Integer|SelectInt 0 _index_4 (== _index_4 _returnvalue_2)) "World") "0, 0,1.000000")) (Variables|Default|GetOceanMaskOuterOffset))) "World" true)
                    (Spline|SetTangentsatSplinePoint (LocalVariables|GetSplineWaterCutOut) _index_4 (Spline|GetArriveTangentatSplinePoint _spline_10 (Math|Integer|SelectInt 0 _index_4 (== _index_4 _returnvalue_2)) "World") (Spline|GetLeaveTangentatSplinePoint _spline_10 (Math|Integer|SelectInt 0 _index_4 (== _index_4 _returnvalue_2)) "World") "World" true)))
                (:then_1
                  (Spline|AddSplinePoint (LocalVariables|GetSplineWaterCutOut) _returnvalue_28 "World" true)
                  (Spline|SetTangentsatSplinePoint (LocalVariables|GetSplineWaterCutOut) 0 (Spline|GetArriveTangentatSplinePoint (Variables|Default|GetSpline) 0 "World") (Spline|GetLeaveTangentatSplinePoint (Variables|Default|GetSpline) 0 "World") "World" true)
                  (LocalVariables|SetMaxDistance (Math|Float|Max(Float) (Math|Float|Absolute(Float) _x_6) (Math|Float|Absolute(Float) _y_4)))
                  (LocalVariables|SetZHeight (.z _returnvalue_28))
                  (LocalVariables|SetX_Sign (Math|Float|SelectFloat _returnvalue_29 (Math|Float|Sign(Float) _x_6) _returnvalue_45))
                  (LocalVariables|SetY_Sign (Math|Float|SelectFloat _returnvalue_29 (* _returnvalue_29 -1.0) _returnvalue_45)))
                (:then_2
                  (Utilities|FlowControl|Sequence
                    (:then_0
                      _)
                    (:then_1
                      (Spline|AddSplinePoint _returnvalue (Math|Vector|MakeVector (* (LocalVariables|GetMaxDistance) (* (LocalVariables|GetX_Sign) -1.0)) (* (LocalVariables|GetMaxDistance) (LocalVariables|GetY_Sign)) (LocalVariables|GetZHeight)) "World" true))
                    (:then_2
                      (Spline|AddSplinePoint _returnvalue (Math|Vector|MakeVector (* (LocalVariables|GetMaxDistance) (* (LocalVariables|GetX_Sign) -1.0)) (* (LocalVariables|GetMaxDistance) (* (LocalVariables|GetY_Sign) -1.0)) (LocalVariables|GetZHeight)) "World" true)))
                  (Spline|SetTangentsatSplinePoint _returnvalue (- (Spline|GetNumberOfSplinePoints _returnvalue) 1) (Math|Vector|MakeVector 0.0) (Math|Vector|MakeVector 0.0) "World" true))
                (:then_3
                  (Spline|AddSplinePoint _returnvalue (- (Spline|GetLocationandTangentatSplinePoint _spline_6 _returnvalue_19 "World") (* (Math|Vector|Normalize (Math|Vector|CrossProduct _returnvalue_20 "0, 0,1.000000")) (+ (Variables|Default|GetOceanMaskOuterOffset) (Variables|Default|GetOceanMaskOuterCutoutOffset)))) "World" true)
                  (Spline|SetTangentatSplinePoint _returnvalue 0 _returnvalue_20 "World" true))
                (:then_4
                  (bind _aswater_body_4 (Utilities|Casting|CastToWaterBody _oceanwatercutout_2))
                  (Water|OnWaterBodyChanged (Variables|/Script/Water.WaterBody|GetWaterBodyComponent _aswater_body_4) true true false))))))))))

(fn UpdateOceanMaskSplinePoints_OLD
  (bind _oceanwatercutout (Variables|Default|GetOceanWaterCutOut))
  (bind _returnvalue (Actor|GetComponentbyClass _oceanwatercutout "/Script/Water.WaterSplineComponent"))
  (bind _spline (Variables|Default|GetSpline))
  (bind _returnvalue_2 (Spline|GetNumberOfSplinePoints _spline))
  (bind _spline_2 (Variables|Default|GetSpline))
  (bind _returnvalue_3 (Spline|GetTangentatSplinePoint _spline_2 0 "World"))
  (bind _oceanwateredgemaskactor (Variables|Default|GetOceanWaterEdgeMaskActor))
  (bind _returnvalue_4 (Actor|GetComponentbyClass _oceanwateredgemaskactor "/Script/Water.WaterSplineComponent"))
  (bind _spline_3 (Variables|Default|GetSpline))
  (bind _returnvalue_5 (Spline|GetNumberOfSplinePoints _spline_3))
  (bind _spline_4 (Variables|Default|GetSpline))
  (bind _returnvalue_6 (Spline|GetNumberOfSplinePoints _spline_4))
  (bind _spline_5 (Variables|Default|GetSpline))
  (bind _location (Spline|GetLocationandTangentatSplinePoint _spline_5 0 "World"))
  (bind _returnvalue_7 (Spline|GetTangentatSplinePoint _spline_5 0 "World"))
  (bind _returnvalue_8 (Math|Vector|CrossProduct _returnvalue_7 "0, 0,1.000000"))
  (bind _returnvalue_9 (Math|Vector|Normalize _returnvalue_8))
  (bind _oceanmaskoutercutoutoffset (Variables|Default|GetOceanMaskOuterCutoutOffset))
  (bind _oceanmaskouteroffset (Variables|Default|GetOceanMaskOuterOffset))
  (bind _returnvalue_10 (+ _oceanmaskouteroffset _oceanmaskoutercutoutoffset))
  (bind _returnvalue_11 (* _returnvalue_9 _returnvalue_10))
  (bind _returnvalue_12 (- _location _returnvalue_11))
  (bind _x (.x _returnvalue_12))
  (bind _y (.y _returnvalue_12))
  (bind _returnvalue_13 (Math|Float|Sign(Float) _y))
  (bind _returnvalue_14 (> _x _y))
  (bind _splinewateredgemask (LocalVariables|GetSplineWaterEdgeMask))
  (bind _spline_6 (Variables|Default|GetSpline))
  (bind _spline_7 (Variables|Default|GetSpline))
  (Utilities|FlowControl|Sequence
    (:then_0
      (Utilities|IsValid (Variables|Default|GetOceanWaterEdgeMaskActor)
        (:Is%20Valid
          (Utilities|IsValid _returnvalue_4
            (:Is%20Valid
              (LocalVariables|SetSplineWaterEdgeMask _returnvalue_4)
              (Transformation|SetActorLocation _oceanwateredgemaskactor "0, 0, 0" false false)
              (UEFNConversionHelper|SetComponentHiddenEditor _returnvalue_4 (not (Variables|Default|GetDebugSplines)))
              (Utilities|FlowControl|Sequence
                (:then_0
                  (Spline|ClearSplinePoints _returnvalue_4 true)
                  (for _index (range (+ _returnvalue_5 1))
                    (Spline|AddSplinePoint (LocalVariables|GetSplineWaterEdgeMask) (+ (Spline|GetLocationandTangentatSplinePoint _spline_6 (Math|Integer|SelectInt 0 _index (== _index _returnvalue_5)) "World") (* (Math|Vector|Normalize (Math|Vector|CrossProduct (Spline|GetTangentatSplinePoint (Variables|Default|GetSpline) (Math|Integer|SelectInt 0 _index (== _index _returnvalue_5)) "World") "0, 0,1.000000")) (Variables|Default|GetOceanMaskInnerOffset))) "World" true)
                    (Spline|SetTangentsatSplinePoint (LocalVariables|GetSplineWaterEdgeMask) _index (Spline|GetArriveTangentatSplinePoint _spline_6 (Math|Integer|SelectInt 0 _index (== _index _returnvalue_5)) "World") (Spline|GetLeaveTangentatSplinePoint _spline_6 (Math|Integer|SelectInt 0 _index (== _index _returnvalue_5)) "World") "World" true)))
                (:then_1
                  (for _index_2 (range (+ _returnvalue_6 1))
                    (Spline|AddSplinePoint _splinewateredgemask (- (Spline|GetLocationandTangentatSplinePoint (Variables|Default|GetSpline) (- _returnvalue_6 _index_2) "World") (* (Math|Vector|Normalize (Math|Vector|CrossProduct (Spline|GetTangentatSplinePoint (Variables|Default|GetSpline) (- _returnvalue_6 _index_2) "World") "0, 0,1.000000")) (Variables|Default|GetOceanMaskOuterOffset))) "World" true)
                    (Spline|SetTangentsatSplinePoint (LocalVariables|GetSplineWaterEdgeMask) (- (Spline|GetNumberOfSplinePoints _splinewateredgemask) 1) (Math|Vector|NegateVector (Spline|GetLeaveTangentatSplinePoint (Variables|Default|GetSpline) (- _returnvalue_6 _index_2) "World")) (Math|Vector|NegateVector (Spline|GetArriveTangentatSplinePoint (Variables|Default|GetSpline) (- _returnvalue_6 _index_2) "World")) "World" true)))
                (:then_2
                  (bind _aswater_body (Utilities|Casting|CastToWaterBody _oceanwateredgemaskactor))
                  (Water|OnWaterBodyChanged (Variables|/Script/Water.WaterBody|GetWaterBodyComponent _aswater_body) true true false))))))))
    (:then_1
      (Utilities|IsValid (Variables|Default|GetOceanWaterCutOut)
        (:Is%20Valid
          (Utilities|IsValid _returnvalue
            (:Is%20Valid
              (LocalVariables|SetSplineWaterCutOut _returnvalue)
              (Transformation|SetActorLocation _oceanwatercutout "0, 0, 0" false false)
              (UEFNConversionHelper|SetComponentHiddenEditor _returnvalue (not (Variables|Default|GetDebugSplines)))
              (Utilities|FlowControl|Sequence
                (:then_0
                  (Spline|ClearSplinePoints (LocalVariables|GetSplineWaterCutOut) true)
                  (for _index_3 (range (+ _returnvalue_2 1))
                    (Spline|AddSplinePoint (LocalVariables|GetSplineWaterCutOut) (- (Spline|GetLocationandTangentatSplinePoint _spline_7 (Math|Integer|SelectInt 0 _index_3 (== _index_3 _returnvalue_2)) "World") (* (Math|Vector|Normalize (Math|Vector|CrossProduct (Spline|GetTangentatSplinePoint (Variables|Default|GetSpline) (Math|Integer|SelectInt 0 _index_3 (== _index_3 _returnvalue_2)) "World") "0, 0,1.000000")) (Variables|Default|GetOceanMaskOuterOffset))) "World" true)
                    (Spline|SetTangentsatSplinePoint (LocalVariables|GetSplineWaterCutOut) _index_3 (Spline|GetArriveTangentatSplinePoint _spline_7 (Math|Integer|SelectInt 0 _index_3 (== _index_3 _returnvalue_2)) "World") (Spline|GetLeaveTangentatSplinePoint _spline_7 (Math|Integer|SelectInt 0 _index_3 (== _index_3 _returnvalue_2)) "World") "World" true)))
                (:then_1
                  (Spline|AddSplinePoint (LocalVariables|GetSplineWaterCutOut) _returnvalue_12 "World" true)
                  (Spline|SetTangentsatSplinePoint (LocalVariables|GetSplineWaterCutOut) 0 (Spline|GetArriveTangentatSplinePoint (Variables|Default|GetSpline) 0 "World") (Spline|GetLeaveTangentatSplinePoint (Variables|Default|GetSpline) 0 "World") "World" true)
                  (LocalVariables|SetMaxDistance (Math|Float|Max(Float) (Math|Float|Absolute(Float) _x) (Math|Float|Absolute(Float) _y)))
                  (LocalVariables|SetZHeight (.z _returnvalue_12))
                  (LocalVariables|SetX_Sign (Math|Float|SelectFloat _returnvalue_13 (Math|Float|Sign(Float) _x) _returnvalue_14))
                  (LocalVariables|SetY_Sign (Math|Float|SelectFloat _returnvalue_13 (* _returnvalue_13 -1.0) _returnvalue_14)))
                (:then_2
                  (Utilities|FlowControl|Sequence
                    (:then_0
                      _)
                    (:then_1
                      (Spline|AddSplinePoint _returnvalue (Math|Vector|MakeVector (* (LocalVariables|GetMaxDistance) (* (LocalVariables|GetX_Sign) -1.0)) (* (LocalVariables|GetMaxDistance) (LocalVariables|GetY_Sign)) (LocalVariables|GetZHeight)) "World" true))
                    (:then_2
                      (Spline|AddSplinePoint _returnvalue (Math|Vector|MakeVector (* (LocalVariables|GetMaxDistance) (* (LocalVariables|GetX_Sign) -1.0)) (* (LocalVariables|GetMaxDistance) (* (LocalVariables|GetY_Sign) -1.0)) (LocalVariables|GetZHeight)) "World" true))
                    (:then_3
                      (Spline|AddSplinePoint _returnvalue (Math|Vector|MakeVector (* (LocalVariables|GetMaxDistance) (LocalVariables|GetX_Sign)) (* (LocalVariables|GetMaxDistance) (* (LocalVariables|GetY_Sign) -1.0)) (LocalVariables|GetZHeight)) "World" true))
                    (:then_4
                      (Spline|AddSplinePoint _returnvalue (Math|Vector|MakeVector (* (LocalVariables|GetMaxDistance) (LocalVariables|GetX_Sign)) (* (LocalVariables|GetMaxDistance) (LocalVariables|GetY_Sign)) (LocalVariables|GetZHeight)) "World" true)))
                  (Spline|SetTangentsatSplinePoint _returnvalue (- (Spline|GetNumberOfSplinePoints _returnvalue) 1) (Math|Vector|MakeVector 0.0) (Math|Vector|MakeVector 0.0) "World" true))
                (:then_3
                  (Spline|AddSplinePoint _returnvalue (- (Spline|GetLocationandTangentatSplinePoint _spline_2 0 "World") (* (Math|Vector|Normalize (Math|Vector|CrossProduct _returnvalue_3 "0, 0,1.000000")) (+ (Variables|Default|GetOceanMaskOuterOffset) (Variables|Default|GetOceanMaskOuterCutoutOffset)))) "World" true)
                  (Spline|SetTangentatSplinePoint _returnvalue 0 _returnvalue_3 "World" true))
                (:then_4
                  (bind _aswater_body_2 (Utilities|Casting|CastToWaterBody _oceanwatercutout))
                  (Water|OnWaterBodyChanged (Variables|/Script/Water.WaterBody|GetWaterBodyComponent _aswater_body_2) true true false))))))))))

(fn ConstructionScript
  (CallFunction|AutoTangentSpline)
  (CallFunction|AlginSplinetoOcean)
  (Default|AddSplineMeshComponents true)
  (Transformation|SetWorldLocation (Variables|Default|GetSceneCaptureComponent2D) "0, 0,50000.000000" false false))
