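# M_CutOutMask is an unlit, Nanite-enabled material in the Asteria water edge environment set.
# Its graph holds a single constant of one, and no connection to a material output is shown.
# The name suggests it marks cut-out mask areas along water edges, though the graph itself contains no masking logic.
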Material /Game/Athena/Asteria/Environments/Water/WaterEdge/M_CutOutMask
graph {
 properties {bUsedWithNanite=true, pixelDepthOffsetMode=PDOM_Legacy, shadingModel=MSM_Unlit, usedShadingModels=Unlit}
 nodes [Constant() {r=1}]
}
