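# Spline-mesh material for water-edge cliffs, blending Apollo cliff rock with tinted water and animated foam.
# A cliff mask and large-scale world XY noise drive the colour, normal and roughness blends between rock and water.
# Panning UV foam and vertically scrolling world-aligned foam use HeightLerp breakup, with a time-of-day cloud tint on the latter.
# Vertex red, mask and noise also scale a large downward offset.
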Material /Game/Athena/Asteria/Environments/Water/WaterEdge/M_WaterEdgeCliff
graph {
 properties {bUsedWithSplineMeshes=true, pixelDepthOffsetMode=PDOM_Legacy, usedShadingModels=Default Lit}
 nodes [
  n1 = WorldPosition()
  n2 = TextureSample(UVs=Divide(A=ComponentMask(None=@n1.XYZ) {g=true, r=true}, B=ScalarParameter() {defaultValue=79682.875, parameterName=WorldNoise_Size})) {texture=/Game/Global/Textures/Default/Noise/T_Tiling_Noise_01}
  n3 = Saturate(None=LinearInterpolate(
    A=ScalarParameter() {defaultValue=-0.4496, parameterName=NoiseContrast_1}
    B=ScalarParameter() {defaultValue=10.657, parameterName=NoiseContrast_2}
    Alpha=@n2.R
   ))
  Add(A=@n3)
  n4 = TextureSample() {samplerType=SAMPLERTYPE_Masks, texture=&/WaterEdgeCliff_Mask}
  n5 = Multiply(A=@n4.R, B=@n4.R)
  n6 = ScalarParameter() {defaultValue=0.9053, parameterName=FoamBlend1}
  n7 = Multiply(A=OneMinus(None=@n4.R), B=@n4.G)
  n8 = ScalarParameter() {defaultValue=0.7852, parameterName=FoamBlend2}
  n9 = Multiply(
   A=Multiply(A=ScalarParameter() {defaultValue=0.36, parameterName=Foam}, B=@n3)
   B=Add(A=Power(Base=@n5, Exp=@n6), B=Power(Base=@n7, Exp=@n8))
  )
  n10 = TextureCoordinate() {coordinateIndex=1}
  n11 = AppendVector(A=ScalarParameter() {defaultValue=3, parameterName=Foam_X}, B=ScalarParameter() {defaultValue=3, parameterName=Foam_y})
  n12 = ScalarParameter() {parameterName=Foam_Speed_x}
  n13 = ScalarParameter() {defaultValue=-0.05, parameterName=Foam_Speed_y}
  n14 = TextureSample(UVs=Panner(Coordinate=Multiply(A=@n10, B=@n11), Speed=AppendVector(A=@n12, B=@n13))) {texture=/FortWater/Textures/T_Foam_DistanceField_Fade_02}
  n15 = ScalarParameter() {defaultValue=0.3346, parameterName=FoamContrast}
  n16 = LinearInterpolate(
   A=VectorParameter() {parameterName=Color_3, defaultValue=<rgba 0.776 0.5117 0.3451 1>}.RGB
   B=VectorParameter() {parameterName=Color_4, defaultValue=<rgba 0.2616 0.346 0.4375 1>}.RGB
   Alpha=@n4.B
  )
  n17 = Multiply(A=TextureCoordinate() {coordinateIndex=1}, B=ScalarParameter() {defaultValue=2, parameterName=Rock_Scale})
  n18 = TextureSample(UVs=@n17) {texture=/Game/Environments/Apollo/Sets/Terrain/Textures/T_Apollo_Cliff_D_01}
  n19 = Multiply(A=@n16, B=@n18.RGB)
  n20 = VectorParameter() {parameterName=Color_1, defaultValue=<rgba 0 0.4773 0.6354 1>}
  n21 = VectorParameter() {parameterName=Color_2, defaultValue=<rgba 0.5152 0.9519 1 1>}
  n22 = Saturate(None=Add(A=@n4.R, B=@n4.R))
  n23 = Multiply(A=LinearInterpolate(A=@n16, B=LinearInterpolate(A=@n20.RGB, B=@n21.RGB, Alpha=@n4.B), Alpha=@n22), B=@n18.RGB)
  n24 = Multiply(A=OneMinus(None=@n4.R), B=Power(Base=@n4.G, Exp=ScalarParameter() {defaultValue=0.5352, parameterName=Contrast}))
  n25 = ScalarParameter() {defaultValue=3.1689, parameterName=Water_DetailFoamScale_X}
  n26 = ScalarParameter() {defaultValue=0.6464, parameterName=Water_DetailFoamScale_Y}
  n27 = Panner(
   Coordinate=Multiply(A=TextureCoordinate() {coordinateIndex=1}, B=AppendVector(A=@n25, B=@n26))
   Speed=AppendVector(A=Constant(), B=ScalarParameter() {defaultValue=-0.028, parameterName=Water_Foam_Speed2})
  )
  n28 = TextureSample(UVs=@n27) {texture=&T_Tiling_Noise_01}
  n29 = ScalarParameter() {parameterName=ContrastTransition}
  n30 = ScalarParameter() {defaultValue=23099.8281, parameterName=Water_FoamScale_X}
  n31 = ScalarParameter() {defaultValue=87329.6797, parameterName=Water_FoamScale_Y}
  n32 = Time()
  n33 = ScalarParameter() {defaultValue=0.05, parameterName=Water_Foam_Speed}
  n34 = Frac(None=AppendVector(A=Constant2Vector(), B=Multiply(A=@n32, B=@n33)))
  n35 = MaterialFunctionCall(
   TextureObject=TextureObject() {texture=/Game/Packages/Fortnite_UniqueMaterials_1/Textures/FoamHeight}
   TextureSize=AppendVector(A=AppendVector(A=@n30, B=@n30), B=@n31)
   WorldPosition=Add(A=WorldPosition().XYZ, B=Multiply(A=@n34, B=@n31))
  ) {materialFunction=&WorldAlignedTexture}
  n36 = Saturate(None=Add(A=ScalarParameter() {defaultValue=0.56, parameterName=DetailOpacity}, B=@n28.R))
  n37 = LinearInterpolate(
   A=Constant()
   B=Multiply(
    A=MaterialFunctionCall(Transition Phase=@n24, Height Texture=@n28.R, Contrast=@n29) {materialFunction=&HeightLerp}.Alpha
    B=Multiply(A=@n35.XY Texture, B=@n36)
   )
   Alpha=@n3
  )
  Add(
   A=Multiply(
    A=MaterialFunctionCall(Transition Phase=@n9, Height Texture=@n14.R, Contrast=@n15) {materialFunction=&HeightLerp}.Alpha
    B=ScalarParameter() {defaultValue=0.872, parameterName=FoamBrightness}
   )
   B=Add(A=LinearInterpolate(A=@n19, B=@n23, Alpha=@n3), B=@n37)
  )
  LinearInterpolate()
  n38 = Multiply(A=VertexColor().R, B=Multiply(A=OneMinus(None=@n22), B=@n2.R))
  LinearInterpolate(A=ScalarParameter() {parameterName=Specular_Bottom}, Alpha=Saturate(None=@n38)) {constA=0.5}
  LinearInterpolate(
   A=MaterialFunctionCall(
    Normal=TextureSample(UVs=@n17) {samplerType=SAMPLERTYPE_Normal, texture=/Game/Environments/Apollo/Sets/Terrain/Textures/T_Apollo_Cliff_N_01}.RGB
    Flatness=ScalarParameter() {defaultValue=0.024, parameterName=Normal_Flatten}
   ) {materialFunction=&FlattenNormal}.Result
   B=Constant3Vector() {constant=<rgba 0 0 1 0>}
   Alpha=Multiply(A=@n3, B=@n22)
  )
  LinearInterpolate(
   A=ScalarParameter() {defaultValue=1.1541, parameterName=Roughness_Rock}
   B=LinearInterpolate(
    A=ScalarParameter() {defaultValue=-4.3753, parameterName=Roughness_Water}
    B=ScalarParameter() {defaultValue=2.326, parameterName=Roughness_foam}
    Alpha=@n4.G
   )
   Alpha=@n3
  )
  MaterialFunctionCall() {materialFunction=/Game/Weapons/WeaponSkins/Wraps/MaterialFunctions/WorldAlignedTexture}
  MaterialFunctionCall!(TextureObject, TextureSize, WorldPosition, Export Float 4, In, World Space Normal, ProjectionTransitionContrast) {materialFunction=/Game/Athena/SafeZone/WorldAlignedTexture_ScalePan}
  Multiply(A=@n38, B=VectorParameter() {parameterName=ZOffset, defaultValue=<rgba 0 0 -24869.9863 1>}.RGB)
  Multiply(
   A=@n37
   B=MaterialFunctionCall(
    Color_Morning=VectorParameter() {parameterName=CloudColor_Morning, defaultValue=<rgba 0.4479 0.2203 0.147 1>}.RGB
    Color_Day=VectorParameter() {parameterName=CloudColor_Day, defaultValue=<rgba 0.419 0.4979 0.599 1>}.RGB
    Color_Evening=VectorParameter() {parameterName=CloudColor_Evening, defaultValue=<rgba 0.401 0.2468 0.2381 1>}.RGB
    Color_Night=VectorParameter() {parameterName=CloudColor_Night, defaultValue=<rgba 0.0312 0.1331 0.6927 1>}.RGB
   ) {materialFunction=&Color_Time_of_Day}.Result
  )
  OneMinus(None=@n4.R)
  OneMinus(None=@n5)
  ScalarParameter() {defaultValue=0.8, parameterName=Emissive}
  TextureSample() {samplerType=SAMPLERTYPE_Normal, texture=&/WaterEdgeCliff_normal}
  TextureSample(UVs=@n27) {samplerType=SAMPLERTYPE_LinearColor, texture=/Game/Athena/Apollo/Water/Textures/01/WaterFoam_FrontFoam_H}
  VectorParameter() {parameterName=ColorEmissive, defaultValue=<rgba 0.419 0.4979 0.599 1>}
 ]
}
