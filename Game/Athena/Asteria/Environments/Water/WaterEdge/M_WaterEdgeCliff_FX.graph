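# Additive unlit spline-mesh material for a water-edge cliff effect, apparently falling foam or mist.
# Two panning foam textures and world-space noise supply detail, alongside UV edge falloff, HeightLerp and depth fade for opacity shaping.
# Vertex offset math bends along normals, stretches downward by noise, and thickens the strip using a camera-to-tangent cross product.
# Base color blends two palettes, each driven by a time-of-day function.
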Material /Game/Athena/Asteria/Environments/Water/WaterEdge/M_WaterEdgeCliff_FX
graph {
 properties {bUsedWithSplineMeshes=true, blendMode=BLEND_Additive, pixelDepthOffsetMode=PDOM_Legacy, shadingModel=MSM_Unlit, usedShadingModels=Unlit}
 nodes [
  DepthFade(Opacity=NamedRerouteUsage(), FadeDistance=ScalarParameter() {defaultValue=8834.1191, parameterName=DepthFade})
  LinearInterpolate(B=Add(A=NamedRerouteUsage(), B=NamedRerouteUsage()), Alpha=NamedRerouteUsage())
  n1 = TextureCoordinate()
  n2 = Power(Base=OneMinus(None=ComponentMask(None=@n1) {g=true}), Exp=ScalarParameter() {defaultValue=0.2592, parameterName=FalloffOpacity})
  n3 = ComponentMask(None=@n1) {r=true}
  n4 = Multiply(A=Multiply(A=@n3, B=OneMinus(None=@n3))) {constB=4}
  n5 = ScalarParameter() {defaultValue=0.6648, parameterName=EdgeFalloff}
  n6 = Multiply(A=@n2, B=Saturate(None=Power(Base=@n4, Exp=@n5)))
  n7 = ScalarParameter() {defaultValue=-1.5894, parameterName=C_1}
  n8 = ScalarParameter() {defaultValue=2.3447, parameterName=C_2}
  n9 = NamedRerouteUsage()
  n10 = Multiply(A=ComponentMask(None=NamedRerouteUsage()) {r=true}, B=Saturate(None=LinearInterpolate(A=@n7, B=@n8, Alpha=@n9)))
  n11 = ScalarParameter() {parameterName=Contrast}
  Multiply(
   A=Multiply(
    A=NamedRerouteUsage()
    B=MaterialFunctionCall(Transition Phase=@n6, Height Texture=@n10, Contrast=@n11) {materialFunction=&HeightLerp}.Alpha
   )
   B=@n2
  )
  n12 = Power(Base=ComponentMask(None=TextureCoordinate()) {g=true}, Exp=ScalarParameter() {defaultValue=0.7367, parameterName=FalloffBend})
  n13 = Multiply(A=@n12, B=VertexNormalWS())
  n14 = ScalarParameter() {defaultValue=19719.1602, parameterName=BendOffset}
  n15 = ScalarParameter() {defaultValue=85805.2109, parameterName=ZOffset}
  n16 = ScalarParameter() {defaultValue=18207.5234, parameterName=ZOffset2}
  n17 = Saturate(None=LinearInterpolate(
    A=ScalarParameter() {defaultValue=1.5832, parameterName=NoiseContrast_1z}
    B=ScalarParameter() {defaultValue=-1.0083, parameterName=NoiseContrast_2z}
    Alpha=NamedRerouteUsage()
   ))
  n18 = Multiply(A=@n12, B=Multiply(A=Constant3Vector() {constant=<rgba 0 0 -1 0>}, B=LinearInterpolate(A=@n15, B=@n16, Alpha=@n17)))
  NamedRerouteDeclaration(None=Add(A=Multiply(A=@n13, B=@n14), B=Multiply(A=@n12, B=@n18))) {name=WPOBendandStretch, nodeColor=<rgba 0 1 0.8359 1>}
  n19 = CameraPositionWS()
  n20 = Add(A=NamedRerouteUsage(), B=WorldPosition().XYZ)
  n21 = Normalize(VectorInput=CrossProduct(A=Subtract(A=@n19, B=@n20), B=MaterialFunctionCall() {materialFunction=&TangentBasis}.Y V))
  n22 = TextureCoordinate()
  n23 = ComponentMask(None=@n22) {g=true}
  n24 = Sign(None=Subtract(A=ComponentMask(None=@n22) {r=true}) {constB=0.5})
  n25 = Multiply(
   A=LinearInterpolate(A=ScalarParameter() {parameterName=WidthUpper_Tangnet}, B=ScalarParameter() {defaultValue=-11202.3887, parameterName=WidthLower}, Alpha=@n23)
   B=@n24
  )
  n26 = Transform(None=Constant3Vector() {constant=<rgba 1 0 0 0>})
  n27 = Multiply(A=ScalarParameter() {defaultValue=511.4523, parameterName=WidthUpper}, B=Normalize(VectorInput=Multiply(A=@n26, B=@n24)))
  n28 = OneMinus(None=@n23)
  NamedRerouteDeclaration(None=Add(A=Multiply(A=@n21, B=@n25), B=Multiply(A=@n27, B=@n28))) {name=Splinethicken, nodeColor=<rgba 0.6875 0 1 1>}
  n29 = ScalarParameter() {defaultValue=-1.0954, parameterName=NoiseContrast_1}
  n30 = ScalarParameter() {defaultValue=10.657, parameterName=NoiseContrast_2}
  n31 = WorldPosition() {worldPositionShaderOffset=WPT_ExcludeAllShaderOffsets}
  n32 = TextureSample(UVs=Divide(A=ComponentMask(None=@n31.XYZ) {g=true, r=true}, B=ScalarParameter() {defaultValue=79682.875, parameterName=WorldNoise_Size})) {texture=/Game/Global/Textures/Default/Noise/T_Tiling_Noise_01}
  NamedRerouteDeclaration(None=Saturate(None=LinearInterpolate(A=@n29, B=@n30, Alpha=@n32.R))) {name=NoiseHighContrast, nodeColor=<rgba 0 1 0.0156 1>}
  NamedRerouteDeclaration(None=@n32.R) {name=Noise, nodeColor=<rgba 1 0.6797 0 1>}
  n33 = ScalarParameter() {parameterName=OpacityContrast_1}
  n34 = ScalarParameter() {defaultValue=2.7377, parameterName=OpacityContrast_2}
  NamedRerouteDeclaration(None=Saturate(None=LinearInterpolate(A=@n33, B=@n34, Alpha=@n32.R))) {name=NoiseForOpacity, nodeColor=<rgba 0 1 0.0625 1>}
  n35 = Multiply(
   A=TextureCoordinate() {coordinateIndex=1}
   B=AppendVector(A=ScalarParameter() {defaultValue=0.4214, parameterName=UV2_X}, B=ScalarParameter() {defaultValue=1.5057, parameterName=UV2_Y})
  )
  n36 = AppendVector(A=Constant(), B=ScalarParameter() {defaultValue=-0.04, parameterName=Speed2})
  NamedRerouteDeclaration(None=TextureSample(UVs=Panner(Coordinate=@n35, Speed=@n36)) {texture=/Game/Packages/Fortnite_UniqueMaterials_1/Textures/FoamHeight}.R) {name=Foamlarge, nodeColor=<rgba 1 0 0.6328 1>}
  n37 = Multiply(
   A=TextureCoordinate() {coordinateIndex=1}
   B=AppendVector(A=ScalarParameter() {defaultValue=0.9814, parameterName=UV_X}, B=ScalarParameter() {defaultValue=1.6069, parameterName=UV_Y})
  )
  n38 = AppendVector(A=Constant(), B=ScalarParameter() {defaultValue=-0.05, parameterName=Speed})
  NamedRerouteDeclaration(None=TextureSample(UVs=Panner(Coordinate=@n37, Speed=@n38)) {texture=/FortWater/Textures/T_Foam_DistanceField_Fade_02}.R) {name=FoamSmall, nodeColor=<rgba 0.5 0 1 1>}
  n39 = VectorParameter() {parameterName=CloudColor_Morning, defaultValue=<rgba 0.276 0.2206 0.2027 1>}
  n40 = VectorParameter() {parameterName=CloudColor_Day, defaultValue=<rgba 0.419 0.4979 0.599 1>}
  n41 = VectorParameter() {parameterName=CloudColor_Evening, defaultValue=<rgba 0.2604 0.2135 0.2075 1>}
  n42 = VectorParameter() {parameterName=CloudColor_Night, defaultValue=<rgba 0.1033 0.1323 0.2917 1>}
  n43 = VectorParameter() {parameterName=CloudColor2_Morning, defaultValue=<rgba 0.1042 0.1042 0.1042 1>}
  n44 = VectorParameter() {parameterName=CloudColor2_Day, defaultValue=<rgba 0.1335 0.192 0.2135 1>}
  n45 = VectorParameter() {parameterName=CloudColor2_Evening, defaultValue=<rgba 0.026 0.0223 0.0214 1>}
  n46 = VectorParameter() {parameterName=CloudColor2_Night, defaultValue=<rgba 0.1094 0.1398 0.1875 1>}
  n47 = NamedRerouteUsage()
  NamedRerouteDeclaration(None=LinearInterpolate(
    A=MaterialFunctionCall(Color_Morning=@n39.RGB, Color_Day=@n40.RGB, Color_Evening=@n41.RGB, Color_Night=@n42.RGB) {materialFunction=&Color_Time_of_Day}.Result
    B=MaterialFunctionCall(Color_Morning=@n43.RGB, Color_Day=@n44.RGB, Color_Evening=@n45.RGB, Color_Night=@n46.RGB) {materialFunction=&Color_Time_of_Day}.Result
    Alpha=ComponentMask(None=@n47) {g=true, r=true}
   )) {name=BaseColor, nodeColor=<rgba 0 1 1 1>}
  Reroute()
  VectorParameter() {parameterName=Color, defaultValue=<rgba 0.419 0.4979 0.599 1>}
  VectorParameter() {parameterName=Color2, defaultValue=<rgba 0.1335 0.192 0.2135 1>}
 ]
}
